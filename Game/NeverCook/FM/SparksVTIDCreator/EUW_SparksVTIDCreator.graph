EditorUtilityWidgetBlueprint /Game/NeverCook/FM/SparksVTIDCreator/EUW_SparksVTIDCreator
graphs [
 EventGraph {nodes [
   K2Node_ComponentBoundEvent_0 = CustomEvent!() -> (OutputDelegate, then, PropertyName:Name) {is=title K2Node_CustomEvent}
   > n1 = Sequence()
   > Branch(Condition="Equal (Name)"(A=@K2Node_ComponentBoundEvent_0.PropertyName, B=SparksID))
   Branch(execute=@n1.then_1, Condition="Equal (Name)"(A=@K2Node_ComponentBoundEvent_0.PropertyName, B=UseCustomPath))
   n2 = Get UseCustomPath() -> (UseCustomPath:bool)
   > SetIsEnabled(self=Get SinglePropertyView_131() -> (SinglePropertyView_131:Single Property View Object Reference), bInIsEnabled=@n2)
   > Set Path(Path:String=Select(Option 0:String=/SparksCosmetics/Items/CosmeticVariantTokens, Option 1:String, Index=@n2) -> (ReturnValue:String)) -> (Output_Get:String)
   "On Clicked (EditorUtilityButton_52)"() {is=title K2Node_ComponentBoundEvent}
   > n3 = For Each Loop(Array:Array of Object References=Make Array!(
     "[0]":Object Reference=Get SparksID() -> (SparksID:Sparks Instrument Item Def Base Object Reference)
     "[1]":Object Reference=Get BackpackID() -> (BackpackID:Athena Backpack Item Definition Object Reference)
     "[2]":Object Reference=Get PickaxeID() -> (PickaxeID:Athena Pickaxe Item Definition Object Reference)
    ) -> (Array:Array of Object References)) -> (Array Element:Object Reference)
   > n4 = Set CurrentAsset(CurrentAsset:Object Reference=@n3) -> (Output_Get:Object Reference)
   > Is Valid(InputObject=@n4)
   > "Set 0=Sparks 1=Backpack 2=Pickaxe"("0=Sparks 1=Backpack 2=Pickaxe":int=@n3.Array Index) -> (Output_Get:int=0)
   > n5 = For Loop(FirstIndex=1, LastIndex=Get Style Count() -> (Style Count:int=0))
   > Set Index(Index:int=int - int(A=@n5)) -> (Output_Get:int=0)
   > n6 = Sequence!(execute) -> (then_0, then_1, then_2:Exec)
   > n7 = GetAssetTools()
   n8 = Append!(
    A=VTID_
    B=GetDisplayName(Object=Get CurrentAsset() -> (CurrentAsset:Object Reference))
    C:String=_Style
    D:String="Get (a copy)"(Array=Get LetterArray() -> (LetterArray:Array of Strings), Dimension 1=Get Index() -> (Index:int=0))
   ) -> (ReturnValue)
   n9 = "integer >= integer"(A=Get Index() -> (Index=0), B=1)
   n10 = int * int(A=6, B="Get 0=Guitar 1=Bass 2=Vocals 3=Drum 4=Keyboard"() -> ("0=Guitar 1=Bass 2=Vocals 3=Drum 4=Keyboard":int=0))
   n11 = int * int(A="Get 0=Sparks 1=Backpack 2=Pickaxe"() -> ("0=Sparks 1=Backpack 2=Pickaxe":int=0), B=2)
   > n12 = DuplicateAsset!(
    execute
    self:Asset Tools Interface=@n7
    AssetName:String=@n8
    PackagePath:String=Get Path() -> (Path:String)
    OriginalObject:Object Reference="Get (a copy)"(
     Array:Array of Object References=Get AssetsRef() -> (AssetsRef:Array of Object References)
     Dimension 1=int + int!(A=@n11, B=@n10, C:bool=@n9) -> (ReturnValue)
    ) -> (Output:Object Reference)
   ) -> (then, ReturnValue:Object Reference)
   > n13 = Set Asset(Asset:Object Reference=@n12) -> (Output_Get:Object Reference)
   > SetEditorProperty(Object=@n13, PropertyName=cosmetic_item, PropertyValue:Object Reference=Get CurrentAsset())
   > SetEditorProperty(Object=@n13, PropertyName=bAutoEquipVariant, PropertyValue:bool="Equal (Integer)"(A=Get Index() -> (Index=0)))
   n14 = "integer >= integer"(A=Get Index() -> (Index=0), B=1)
   n15 = Select!(
    Option 0:Text=Select!(
     Option 0="LOCTEXT(\"Guitar\")"
     Option 1="LOCTEXT(\"Mic\")"
     Option 2="LOCTEXT(\"Bass\")"
     Option 3:Text="LOCTEXT(\"Drum\")"
     Option 4:Text="LOCTEXT(\"Keytar\")"
     Index="Get 0=Guitar 1=Bass 2=Vocals 3=Drum 4=Keyboard"() -> ("0=Guitar 1=Bass 2=Vocals 3=Drum 4=Keyboard"=0)
    ) -> (ReturnValue)
    Option 1:Text="LOCTEXT(\"Backpack\")"
    Option 2:Text="LOCTEXT(\"Pickaxe\")"
    Index:int="Get 0=Sparks 1=Backpack 2=Pickaxe"() -> ("0=Sparks 1=Backpack 2=Pickaxe"=0)
   ) -> (ReturnValue:Text)
   > SetEditorProperty(
    Object=@n13
    PropertyName=ItemShortDescription
    PropertyValue:Text=MakeLiteralText(Value=SelectText(A="LOCTEXT(\"Style\")", B=@n15, bPickA=@n14))
   )
   n16 = "Get 0=Sparks 1=Backpack 2=Pickaxe"() -> ("0=Sparks 1=Backpack 2=Pickaxe"=0)
   n17 = RequestGameplayTag!(
    execute=@n6.then_2
    TagName:String=Select!(
     Option 0:String=Cosmetics.Variant.Channel.Sparks.Instrument.Material
     Option 1:String=Cosmetics.Variant.Channel.Particle
     Index:bool="Equal (Integer)"(A=@n16, B=2)
    ) -> (ReturnValue:String)
    bErrorIfNotFound:bool=false
   ) -> (then, OutTag:Gameplay Tag Structure, ReturnValue:bool)
   > Set Tag(Tag:Gameplay Tag Structure=@n17) -> (Output_Get:Gameplay Tag Structure)
   > Branch(Condition=@n17.ReturnValue)
   > SetEditorProperty(
    Object=Get Asset() -> (Asset:Object Reference)
    PropertyName=VariantChannelTag
    PropertyValue:Gameplay Tag Structure=Get Tag() -> (Tag:Gameplay Tag Structure)
   )
   n18 = MakeLiteralString(Value=Cosmetics.Variant.Property.Particle)
   n19 = "To String (Integer)"(InInt=int + int!(A=Get Index() -> (Index=0), B) -> (ReturnValue))
   n20 = "To String (Integer)"(InInt=Get Index() -> (Index=0))
   n21 = MakeLiteralString(Value=Cosmetics.Variant.Property.Mat)
   n22 = "Get 0=Sparks 1=Backpack 2=Pickaxe"() -> ("0=Sparks 1=Backpack 2=Pickaxe"=0)
   n23 = RequestGameplayTag(
    execute=@n6.then_1
    TagName=Select(Option 0=Append!(A=@n21, B=@n20) -> (ReturnValue), Option 1=Append(A=@n18, B=@n19), Index="Equal (Integer)"(A=@n22, B=2))
    bErrorIfNotFound=false
   )
   > Set Tag(Tag=@n23)
   > Branch(Condition=@n23.ReturnValue)
   > SetEditorProperty(Object=Get Asset(), PropertyName=VariantNameTag, PropertyValue=Get Tag())
   "On Clicked (EditorUtilityButton)"() {is=title K2Node_ComponentBoundEvent}
   > SyncBrowserToFolders!(execute, FolderList:Array of Strings=Make Array!("[0]":String=/SparksCosmetics/Instrument) -> (Array:Array of Strings)) -> (then)
   "On Clicked (EditorUtilityButton_60)"() {is=title K2Node_ComponentBoundEvent}
   > Set Path(Path="Get (a copy)"(Array:Array of Strings=GetSelectedFolderPaths!() -> (ReturnValue:Array of Strings)) -> (Output:String))
   "On Property Changed (DetailsView)"!() -> (OutputDelegate, then, PropertyName:Name=None) {is=title K2Node_ComponentBoundEvent}
   > Refresh data()
   n24 = "On Property Changed (DetailsView_55)"!() -> (OutputDelegate, then, PropertyName:Name=None) {is=title K2Node_ComponentBoundEvent}
   > n25 = Sequence!(execute) -> (then_0, then_1)
   > Branch(Condition="Equal (Name)"(A=@n24.PropertyName, B=SparksID))
   Branch(execute=@n25.then_1, Condition="Equal (Name)"(A=@n24.PropertyName, B=UseCustomPath))
   n26 = Get UseCustomPath()
   > SetIsEnabled(self=Get SinglePropertyView_131(), bInIsEnabled=@n26)
   > Set Path(Path=Select(Option 0=/Plugins/FM/SparksCosmetics/Items/CosmeticVariantTokens, Option 1=Get Path(), Index=@n26))
   n27 = "On Check State Changed (EditorUtilityCheckBox_0)"!() -> (OutputDelegate, then, bIsChecked:bool) {is=title K2Node_ComponentBoundEvent}
   > n28 = Set UseCustomPath(UseCustomPath:bool=@n27.bIsChecked) -> (Output_Get:bool)
   > SetIsEnabled(
    self=[Get SinglePropertyView_131(), Get EditorUtilityButton_60() -> (EditorUtilityButton_60:Editor Utility Button Object Reference)]
    bInIsEnabled=@n28
   )
   > Set Path(Path=Select(Option 0=&CosmeticVariantTokens, Option 1=Get Path(), Index=Get UseCustomPath()))
   "On Clicked (EditorUtilityButton_1)"() {is=title K2Node_ComponentBoundEvent}
   > SyncBrowserToFolders(FolderList=Make Array("[0]"=/SparksCosmetics/Items/CosmeticVariantTokens))
   Refresh data() {is=title K2Node_CustomEvent}
   n29 = IsValid(Object=Get SparksID())
   > SetIsEnabled(
    self=[
     Get EditorUtilityButton_52() -> (EditorUtilityButton_52:Editor Utility Button Object Reference)
     Get DetailsView_55() -> (DetailsView_55:Details View Object Reference)
    ]
    bInIsEnabled=@n29
   )
   > GetOtherIDs!(execute, self, SparksID:Sparks Instrument Item Def Base Object Reference=Get SparksID()) -> (then)
   > n30 = Branch(Condition=@n29)
   > n31 = For Each Loop with Break!(
    Exec:Exec
    Array:Array of ESparksInstrumentType Enums=Get InstrumentArray() -> (InstrumentArray:Array of ESparksInstrumentType Enums=Guitar)
    Break:Exec=@n34.then
   ) -> (LoopBody:Exec, Array Element:ESparksInstrumentType Enum, Array Index:int, Completed:Exec)
   n32 = Get SparksID()
   > n33 = Branch(Condition=Contains(
     SearchIn=GetDisplayName(Object=@n32)
     Substring=Select!(
      Guitar:String=_Guitar
      Bass:String=_Bass
      Vocals:String=_Mic
      Drum:String=_Drum
      Keyboard:String=_Keytar
      None:String
      Index:ESparksInstrumentType Enum=@n31
     ) -> (ReturnValue)
    ))
   > Set Instrument(Instrument:ESparksInstrumentType Enum=@n31) -> (Output_Get:ESparksInstrumentType Enum=Guitar)
   > n34 = "Set 0=Guitar 1=Bass 2=Vocals 3=Drum 4=Keyboard"("0=Guitar 1=Bass 2=Vocals 3=Drum 4=Keyboard":int=@n31.Array Index) -> (Output_Get:int=0)
   Set Instrument(execute=@n33.else, Instrument=None) -> (Output_Get=Guitar)
   Set Instrument(execute=@n30.else, Instrument=None) -> (Output_Get=Guitar)
   Event PreConstruct()
   n35 = Get SinglePropertyView_131()
   > SetObject(self=@n35, NewObject=Self-Reference())
   > SetPropertyName!(execute, self:Single Property View Object Reference=@n35, NewPropertyName:Name=Path) -> (then)
   > SetObject(self=[Get DetailsView_55(), Get DetailsView() -> (DetailsView:Details View Object Reference)], NewObject=Self-Reference())
   > Refresh data() {is=title K2Node_CallFunction}
   SelectString(A=/Plugins/FM/SparksCosmetics/Items/CosmeticVariantTokens, B=/Game/Developers/MattMartins/SparksVTIDTool, bPickA=false)
   DoesAssetExist(AssetPath=Append!(A=Get Path(), B=@n8, C:String=., D:String=@n8) -> (ReturnValue))
   n36 = "integer >= integer"(A=Get Index() -> (Index=0), B=1)
   n37 = Get AssetsRef()
   n38 = Get AssetsRef()
   n39 = Get AssetsRef()
   n40 = Get AssetsRef()
   n41 = Get AssetsRef()
   n42 = Select(
    Guitar:Object Reference=Select!(
     Option 0:Object Reference="Get (a copy)"(Array:Array of Object References=@n37) -> (Output:Object Reference)
     Option 1:Object Reference="Get (a copy)"(Array=@n37, Dimension 1=2)
     Option 2:Object Reference="Get (a copy)"(Array=@n37, Dimension 1=4)
     Index:int="Get 0=Sparks 1=Backpack 2=Pickaxe"() -> ("0=Sparks 1=Backpack 2=Pickaxe"=0)
    ) -> (ReturnValue)
    Bass:Object Reference=Select(
     Option 0="Get (a copy)"(Array=@n38, Dimension 1=6)
     Option 1="Get (a copy)"(Array=@n38, Dimension 1=8)
     Option 2="Get (a copy)"(Array=@n38, Dimension 1=10)
     Index="Get 0=Sparks 1=Backpack 2=Pickaxe"() -> ("0=Sparks 1=Backpack 2=Pickaxe"=0)
    )
    Vocals:Object Reference=Select(
     Option 0="Get (a copy)"(Array=@n39, Dimension 1=12)
     Option 1="Get (a copy)"(Array=@n39, Dimension 1=14)
     Option 2="Get (a copy)"(Array=@n39, Dimension 1=16)
     Index="Get 0=Sparks 1=Backpack 2=Pickaxe"() -> ("0=Sparks 1=Backpack 2=Pickaxe"=0)
    )
    Drum:Object Reference=Select(
     Option 0="Get (a copy)"(Array=@n40, Dimension 1=18)
     Option 1="Get (a copy)"(Array=@n40, Dimension 1=20)
     Option 2="Get (a copy)"(Array=@n40, Dimension 1=22)
     Index="Get 0=Sparks 1=Backpack 2=Pickaxe"() -> ("0=Sparks 1=Backpack 2=Pickaxe"=0)
    )
    Keyboard:Object Reference=Select(
     Option 0="Get (a copy)"(Array=@n41, Dimension 1=24)
     Option 1="Get (a copy)"(Array=@n41, Dimension 1=26)
     Option 2="Get (a copy)"(Array=@n41, Dimension 1=28)
     Index="Get 0=Sparks 1=Backpack 2=Pickaxe"() -> ("0=Sparks 1=Backpack 2=Pickaxe"=0)
    )
    None:Object Reference=/SparksCosmetics/Items/CosmeticVariantTokens/_VTID_Backpack_Placeholder_Bass_StyleA
    Index=Get Instrument() -> (Instrument:ESparksInstrumentType Enum=Guitar)
   ) -> (ReturnValue:Object Reference)
   n43 = Select!(
    Guitar:Object Reference=Select!(
     Option 0:Object Reference=/SparksCosmetics/Items/CosmeticVariantTokens/_VTID_Sparks_Placeholder_Guitar_StyleB
     Option 1:Object Reference=/SparksCosmetics/Items/CosmeticVariantTokens/_VTID_Backpack_Placeholder_Guitar_StyleB
     Option 2:Object Reference=/SparksCosmetics/Items/CosmeticVariantTokens/_VTID_Pickaxe_Placeholder_Guitar_StyleB
     Index:int="Get 0=Sparks 1=Backpack 2=Pickaxe"() -> ("0=Sparks 1=Backpack 2=Pickaxe"=0)
    ) -> (ReturnValue)
    Bass:Object Reference=Select(
     Option 0=/SparksCosmetics/Items/CosmeticVariantTokens/_VTID_Sparks_Placeholder_Bass_StyleB
     Option 1=/SparksCosmetics/Items/CosmeticVariantTokens/_VTID_Backpack_Placeholder_Bass_StyleB
     Option 2=/SparksCosmetics/Items/CosmeticVariantTokens/_VTID_Pickaxe_Placeholder_Bass_StyleB
     Index="Get 0=Sparks 1=Backpack 2=Pickaxe"() -> ("0=Sparks 1=Backpack 2=Pickaxe"=0)
    )
    Vocals:Object Reference=Select(
     Option 0=/SparksCosmetics/Items/CosmeticVariantTokens/_VTID_Sparks_Placeholder_Mic_StyleB
     Option 1=/SparksCosmetics/Items/CosmeticVariantTokens/_VTID_Backpack_Placeholder_Mic_StyleB
     Option 2=/SparksCosmetics/Items/CosmeticVariantTokens/_VTID_Pickaxe_Placeholder_Mic_StyleB
     Index="Get 0=Sparks 1=Backpack 2=Pickaxe"() -> ("0=Sparks 1=Backpack 2=Pickaxe"=0)
    )
    Drum:Object Reference=Select(
     Option 0=/SparksCosmetics/Items/CosmeticVariantTokens/_VTID_Sparks_Placeholder_Drum_StyleB
     Option 1=/SparksCosmetics/Items/CosmeticVariantTokens/_VTID_Backpack_Placeholder_Drum_StyleB
     Option 2=/SparksCosmetics/Items/CosmeticVariantTokens/_VTID_Pickaxe_Placeholder_Drum_StyleB
     Index="Get 0=Sparks 1=Backpack 2=Pickaxe"() -> ("0=Sparks 1=Backpack 2=Pickaxe"=0)
    )
    Keyboard:Object Reference=Select(
     Option 0=/SparksCosmetics/Items/CosmeticVariantTokens/_VTID_Sparks_Placeholder_Keytar_StyleB
     Option 1=/SparksCosmetics/Items/CosmeticVariantTokens/_VTID_Backpack_Placeholder_Keytar_StyleB
     Option 2=/SparksCosmetics/Items/CosmeticVariantTokens/_VTID_Pickaxe_Placeholder_Keytar_StyleB
     Index="Get 0=Sparks 1=Backpack 2=Pickaxe"() -> ("0=Sparks 1=Backpack 2=Pickaxe"=0)
    )
    None:Object Reference
    Index:ESparksInstrumentType Enum=Get Instrument() -> (Instrument=Guitar)
   ) -> (ReturnValue)
   Reroute Node(InputPin:Object Reference=Select!(
     Option 0=Select(Option 0=@n42, Option 1=@n43, Index=@n36)
     Option 1=&_VTID_Backpack_Placeholder_Bass_StyleA
     Index:bool=false
    ) -> (ReturnValue)) -> (OutputPin:Object Reference)
   "Set 0=Sparks 1=Backpack 2=Pickaxe"(execute=@n46.else, "0=Sparks 1=Backpack 2=Pickaxe"=-1) -> (Output_Get=0)
   n44 = For Each Loop with Break(Array:Array of Strings=Get TypeAray() -> (TypeAray:Array of Strings), Break=@n47.then) -> (Array Element:String)
   n45 = Contains(SearchIn=GetDisplayName(Object=Get SparksID()), Substring=@n44)
   > n46 = Branch(Condition=@n45)
   > n47 = "Set 0=Sparks 1=Backpack 2=Pickaxe"("0=Sparks 1=Backpack 2=Pickaxe"=@n44.Array Index) -> (Output_Get=0)
  ]}
 GetAssetReferencers {nodes [
   n1 = GetAssetReferencers!() -> (then, AssetData:"Asset Data Structure (by ref)") {is=title K2Node_FunctionEntry}
   > n2 = Get Referencers!(
    execute
    self:Asset Registry Interface=GetAssetRegistry()
    PackageName:Name=Break Asset Data(AssetData=@n1)
    ReferenceOptions:"Asset Registry Dependency Options Structure (by ref)"=Make Asset Registry Dependency Options(bIncludeSoftPackageReferences=True)
   ) -> (then, OutReferencers:Array of Names, ReturnValue:bool) {is=title K2Node_CallFunction}
   > Branch(Condition=Is Valid Index!(TargetArray:Array of Names=@n2, IndexToTest:int) -> (ReturnValue:bool))
   > n3 = GetAssets!(
    execute
    self:Asset Registry Interface=GetAssetRegistry()
    Filter:"ARFilter Structure (by ref)"=MakeARFilter!(
     PackageNames:Array of Names=@n2
     PackagePaths:Array of Names
     SoftObjectPaths:Array of Soft Object Path Structures
     ClassPaths:Array of Top Level Asset Path Structures
     RecursiveClassPathsExclusionSet:Set of Top Level Asset Path Structures
     ClassNames:Array of Names
     RecursiveClassesExclusionSet:Set of Names
     bRecursivePaths:bool=false
     bRecursiveClasses:bool=false
     bIncludeOnlyOnDiskAssets:bool=false
    ) -> (ReturnValue:ARFilter Structure)
    bSkipARFilteredAssets:bool=true
   ) -> (then, OutAssetData:Array of Asset Data Structures, ReturnValue:bool)
   > Return Node!(execute, OutAssetData:Array of Asset Data Structures=@n3)
  ]}
 GetInstrumentMeshes {nodes [
   n1 = GetInstrumentMeshes!() -> (then, SparksID:Sparks Instrument Item Def Base Object Reference) {is=title K2Node_FunctionEntry}
   n2 = Get LeftHandMesh!(self:Sparks Instrument Item Def Base Object Reference=@n1) -> (LeftHandMesh:Skeletal Mesh Soft Object Reference)
   > n3 = For Each Loop(Array:Array of Skeletal Mesh Soft Object References=Make Array!(
     "[0]":Skeletal Mesh Soft Object Reference=Get Mesh(self:Sparks Instrument Item Def Base Object Reference=@n1) -> (Mesh:Skeletal Mesh Soft Object Reference)
     "[1]":Skeletal Mesh Soft Object Reference=Get AuxiliaryMesh!(self:Sparks Instrument Item Def Base Object Reference=@n1) -> (AuxiliaryMesh:Skeletal Mesh Soft Object Reference)
     "[2]":Skeletal Mesh Soft Object Reference=@n2
    ) -> (Array:Array of Skeletal Mesh Soft Object References)) -> (Array Element:Skeletal Mesh Soft Object Reference)
   n4 = Sequence(execute=@n3.Completed)
   > Branch(Condition="Equal (Enum)"(
     A:ESparksInstrumentType Enum=GetInstrumentType!(self:Sparks Account Item Definition Object Reference=@n1) -> (ReturnValue:ESparksInstrumentType Enum=Guitar)
     B:ESparksInstrumentType Enum=Vocals
    ))
   > n5 = LoadAsset_Blocking(Asset=@n2)
   > n6 = GetPathNameForLoadedAsset(LoadedAsset=@n5)
   n7 = Split(SourceString=@n6, InStr=/, SearchDir=FromEnd)
   > n8 = Get Asset By Object Path!(
    execute
    self:Asset Registry Interface=GetAssetRegistry()
    ObjectPath_PathString:String=Append!(A=@n7, B=/, C=Replace(SourceString=@n7.RightS, From=MicStand, To=MicStand_Backpack)) -> (ReturnValue)
    bIncludeOnlyOnDiskAssets:bool=false
    bSkipARFilteredAssets:bool=true
   ) -> (then, ReturnValue:Asset Data Structure) {is=title K2Node_CallFunction}
   > Add Unique(
    TargetArray:Array of Asset Data Structures=Get MeshesList() -> (MeshesList:Array of Asset Data Structures)
    NewItem:"Asset Data Structure (by ref)"=@n8
   )
   Return Node!(execute=@n4.then_1, Meshes:Array of Asset Data Structures=Get MeshesList())
   Branch(execute=@n3.LoopBody, Condition=IsValidSoftObjectReference(SoftObjectReference=@n3))
   > n9 = LoadAsset_Blocking(Asset=@n3)
   > n10 = GetPathNameForLoadedAsset(LoadedAsset=@n9)
   > n11 = Get Asset By Object Path(self=GetAssetRegistry(), ObjectPath_PathString=@n10, bIncludeOnlyOnDiskAssets=false, bSkipARFilteredAssets=true)
   > Add Unique(TargetArray=Get MeshesList(), NewItem=@n11)
  ]}
 GetOtherIDs {nodes [
   n1 = GetOtherIDs!() -> (then, SparksID:Sparks Instrument Item Def Base Object Reference) {is=title K2Node_FunctionEntry}
   > n2 = GetInstrumentMeshes!(execute, self, SparksID:Sparks Instrument Item Def Base Object Reference=@n1) -> (then, Meshes:Array of Asset Data Structures) {is=title K2Node_CallFunction}
   > n3 = For Each Loop(Array:Array of Asset Data Structures=@n2) -> (Array Element:Asset Data Structure)
   Set PickaxeID(
    execute=@n3.Completed
    PickaxeID:Athena Pickaxe Item Definition Object Reference=Get LocalPickaxe() -> (LocalPickaxe:Athena Pickaxe Item Definition Object Reference)
   ) -> (Output_Get:Athena Pickaxe Item Definition Object Reference)
   > Set BackpackID(BackpackID:Athena Backpack Item Definition Object Reference=Get LocalBackpack() -> (LocalBackpack:Athena Backpack Item Definition Object Reference)) -> (Output_Get:Athena Backpack Item Definition Object Reference)
   n4 = GetAssetReferencers!(execute=@n3.LoopBody, self, AssetData:"Asset Data Structure (by ref)"=@n3) -> (then, OutAssetData:Array of Asset Data Structures) {is=title K2Node_CallFunction}
   > n5 = For Each Loop(Array=@n4)
   n6 = GetAsset(InAssetData=@n5)
   > n7 = Cast To AthenaPickaxeItemDefinition!(execute, Object=@n6) -> (then, CastFailed, AsAthena Pickaxe Item Definition:Athena Pickaxe Item Definition Object Reference) {is=title K2Node_DynamicCast}
   > n8 = Set LocalPickaxe(LocalPickaxe:Athena Pickaxe Item Definition Object Reference=@n7) -> (Output_Get:Athena Pickaxe Item Definition Object Reference)
   n9 = GetPathNameForLoadedAsset(execute=[@n16.CastFailed, @n8.then, @n17.then], LoadedAsset=@n6)
   > n10 = Get Asset By Object Path(self=GetAssetRegistry(), ObjectPath_PathString=@n9, bIncludeOnlyOnDiskAssets=false, bSkipARFilteredAssets=true)
   > n11 = GetAssetReferencers(AssetData=@n10)
   > n12 = For Each Loop(Array=@n11)
   n13 = GetAsset(InAssetData=@n12)
   > n14 = Cast To AthenaPickaxeItemDefinition(Object=@n13)
   > Set LocalPickaxe(LocalPickaxe=@n14)
   n15 = Cast To AthenaBackpackItemDefinition!(execute=@n14.CastFailed, Object=@n13) -> (then, CastFailed, AsAthena Backpack Item Definition:Athena Backpack Item Definition Object Reference) {is=title K2Node_DynamicCast}
   > Set LocalBackpack(LocalBackpack:Athena Backpack Item Definition Object Reference=@n15) -> (Output_Get:Athena Backpack Item Definition Object Reference)
   n16 = Cast To AthenaBackpackItemDefinition(execute=@n7.CastFailed, Object=@n6)
   > n17 = Set LocalBackpack(LocalBackpack=@n16)
  ]}
]
variables [
 "0=Guitar 1=Bass 2=Vocals 3=Drum 4=Keyboard" {type=integer, default=-1}
 "0=Sparks 1=Backpack 2=Pickaxe" {type=integer, default=-1}
 Array Element {type=Object, default=null}
 Asset {type=Object, default=null}
 AssetsRef {
  type="array<Object>"
  default [
   /SparksCosmetics/Items/CosmeticVariantTokens/_VTID_Sparks_Placeholder_Guitar_StyleA
   &_VTID_Sparks_Placeholder_Guitar_StyleB
   /SparksCosmetics/Items/CosmeticVariantTokens/_VTID_Backpack_Placeholder_Guitar_StyleA
   &_VTID_Backpack_Placeholder_Guitar_StyleB
   /SparksCosmetics/Items/CosmeticVariantTokens/_VTID_Pickaxe_Placeholder_Guitar_StyleA
   &_VTID_Pickaxe_Placeholder_Guitar_StyleB
   /SparksCosmetics/Items/CosmeticVariantTokens/_VTID_Sparks_Placeholder_Bass_StyleA
   &_VTID_Sparks_Placeholder_Bass_StyleB
   &_VTID_Backpack_Placeholder_Bass_StyleA
   &_VTID_Backpack_Placeholder_Bass_StyleB
   /SparksCosmetics/Items/CosmeticVariantTokens/_VTID_Pickaxe_Placeholder_Bass_StyleA
   &_VTID_Pickaxe_Placeholder_Bass_StyleB
   /SparksCosmetics/Items/CosmeticVariantTokens/_VTID_Sparks_Placeholder_Mic_StyleA
   &_VTID_Sparks_Placeholder_Mic_StyleB
   /SparksCosmetics/Items/CosmeticVariantTokens/_VTID_Backpack_Placeholder_Mic_StyleA
   &_VTID_Backpack_Placeholder_Mic_StyleB
   /SparksCosmetics/Items/CosmeticVariantTokens/_VTID_Pickaxe_Placeholder_Mic_StyleA
   &_VTID_Pickaxe_Placeholder_Mic_StyleB
   /SparksCosmetics/Items/CosmeticVariantTokens/_VTID_Sparks_Placeholder_Drum_StyleA
   &_VTID_Sparks_Placeholder_Drum_StyleB
   /SparksCosmetics/Items/CosmeticVariantTokens/_VTID_Backpack_Placeholder_Drum_StyleA
   &_VTID_Backpack_Placeholder_Drum_StyleB
   /SparksCosmetics/Items/CosmeticVariantTokens/_VTID_Pickaxe_Placeholder_Drum_StyleA
   &_VTID_Pickaxe_Placeholder_Drum_StyleB
   /SparksCosmetics/Items/CosmeticVariantTokens/_VTID_Sparks_Placeholder_Keytar_StyleA
   &_VTID_Sparks_Placeholder_Keytar_StyleB
   /SparksCosmetics/Items/CosmeticVariantTokens/_VTID_Backpack_Placeholder_Keytar_StyleA
   &_VTID_Backpack_Placeholder_Keytar_StyleB
   /SparksCosmetics/Items/CosmeticVariantTokens/_VTID_Pickaxe_Placeholder_Keytar_StyleA
   &_VTID_Pickaxe_Placeholder_Keytar_StyleB
  ]
 }
 BackpackID {type=AthenaBackpackItemDefinition, default=null}
 CurrentAsset {type=Object, default=null}
 Debug {type=string, default=""}
 IgnoreExistingFiles {type=boolean}
 Index {type=integer}
 Instrument {type=ESparksInstrumentType, default=None}
 InstrumentArray {type="array<ESparksInstrumentType>", default [Guitar,Bass,Vocals,Drum,Keyboard]}
 InstrumentTypeArray {type="array<string>", default []}
 LetterArray {type="array<string>", default [A,B,C,D,E,F,G,H,I,J,K,L,M,N,O,P,Q,R,S,T,U,V,W,X,Y,Z]}
 Path {type=string, default=/SparksCosmetics/Items/CosmeticVariantTokens}
 PickaxeID {type=AthenaPickaxeItemDefinition, default=null}
 Placeholders VTIDs Asset Paths {
  type="array<string>"
  default [
   &_VTID_Backpack_Placeholder_Bass_StyleA
   &_VTID_Backpack_Placeholder_Bass_StyleB
   &_VTID_Backpack_Placeholder_Drum_StyleA
   &_VTID_Backpack_Placeholder_Drum_StyleB
   &_VTID_Backpack_Placeholder_Guitar_StyleA
   &_VTID_Backpack_Placeholder_Guitar_StyleB
   &_VTID_Backpack_Placeholder_Keytar_StyleA
   &_VTID_Backpack_Placeholder_Keytar_StyleB
   &_VTID_Backpack_Placeholder_Mic_StyleA
   &_VTID_Backpack_Placeholder_Mic_StyleB
   &_VTID_Pickaxe_Placeholder_Bass_StyleA
   &_VTID_Pickaxe_Placeholder_Bass_StyleB
   &_VTID_Pickaxe_Placeholder_Drum_StyleA
   &_VTID_Pickaxe_Placeholder_Drum_StyleB
   &_VTID_Pickaxe_Placeholder_Guitar_StyleA
   &_VTID_Pickaxe_Placeholder_Guitar_StyleB
   &_VTID_Pickaxe_Placeholder_Keytar_StyleA
   &_VTID_Pickaxe_Placeholder_Keytar_StyleB
   &_VTID_Pickaxe_Placeholder_Mic_StyleA
   &_VTID_Pickaxe_Placeholder_Mic_StyleB
   &_VTID_Sparks_Placeholder_Bass_StyleA
   &_VTID_Sparks_Placeholder_Bass_StyleB
   &_VTID_Sparks_Placeholder_Drum_StyleA
   &_VTID_Sparks_Placeholder_Drum_StyleB
   &_VTID_Sparks_Placeholder_Guitar_StyleA
   &_VTID_Sparks_Placeholder_Guitar_StyleB
   &_VTID_Sparks_Placeholder_Keytar_StyleA
   &_VTID_Sparks_Placeholder_Keytar_StyleB
   &_VTID_Sparks_Placeholder_Mic_StyleA
   &_VTID_Sparks_Placeholder_Mic_StyleB
  ]
 }
 SparksID {type=SparksInstrumentItemDefBase, default=null}
 Style Count {type=integer, default=2}
 Tag {type=GameplayTag, default {tagName=Cosmetics.Variant.Property.Mat2}}
 TypeAray {type="array<string>", default [Sparks,Backpack,Pickaxe]}
 UseCustomPath {type=boolean}
]
widgets {
 info {rootWidgetClass=&VerticalBox, widgetCount=14}
 widgets [
  VerticalBox_29 {widgetClassPath=&VerticalBox, widget=VerticalBox_29, properties {bOverride_Cursor=true}}
  DetailsView {
   widgetClassPath=&DetailsView
   parent=VerticalBox_29
   widget=DetailsView
   properties {propertiesToShow [SparksID, Style Count]}
   slot {class=VerticalBoxSlot}
  }
  EditorUtilityButton {widgetClassPath=&EditorUtilityButton, parent=VerticalBox_29, widget=EditorUtilityButton, slot {class=VerticalBoxSlot}}
  TextBlock {
   widgetClassPath=&TextBlock
   parent=EditorUtilityButton
   widget=TextBlock
   properties {
    text=Browse To Instrument Folder
    font {
     bForceMonospaced=false
     bMaterialIsStencil=false
     fontMaterial=null
     fontObject=&Roboto
     letterSpacing=0
     monospacedWidth=1
     size=12
     skewAmount=0
     typefaceFontName=Bold
     outlineSettings {
      bApplyOutlineToDropShadows=false
      bMiteredCorners=false
      bSeparateFillAlpha=false
      outlineMaterial=null
      outlineSize=0
      outlineColor=<rgba 0 0 0 1>
     }
    }
   }
   slot {class=ButtonSlot}
  }
  DetailsView_55 {
   widgetClassPath=&DetailsView
   parent=VerticalBox_29
   widget=DetailsView_55
   properties {propertiesToShow [Instrument,BackpackID,PickaxeID]}
   slot {class=VerticalBoxSlot}
  }
  HorizontalBox_0 {widgetClassPath=&HorizontalBox, parent=VerticalBox_29, widget=HorizontalBox_0, slot {class=VerticalBoxSlot}}
  EditorUtilityCheckBox_0 {
   widgetClassPath=&EditorUtilityCheckBox
   bIsVariable=true
   parent=HorizontalBox_0
   widget=EditorUtilityCheckBox_0
   slot {class=HorizontalBoxSlot, properties {padding=<ltrb 10 0 0 0>}}
  }
  SinglePropertyView_131 {
   widgetClassPath=&SinglePropertyView
   parent=HorizontalBox_0
   widget=SinglePropertyView_131
   properties {bIsEnabled=false}
   slot {class=HorizontalBoxSlot, properties {size {sizeRule=Fill, value=1}}}
  }
  EditorUtilityButton_60 {
   widgetClassPath=&EditorUtilityButton
   parent=HorizontalBox_0
   widget=EditorUtilityButton_60
   properties {bIsEnabled=false}
   slot {class=HorizontalBoxSlot, properties {padding=<ltrb -20 0 10 0>}}
  }
  TextBlock_76 {
   widgetClassPath=&TextBlock
   parent=EditorUtilityButton_60
   widget=TextBlock_76
   properties {
    text=Fill From Browser
    font {
     bForceMonospaced=false
     bMaterialIsStencil=false
     fontMaterial=null
     fontObject=&Roboto
     letterSpacing=0
     monospacedWidth=1
     size=12
     skewAmount=0
     typefaceFontName=Bold
     outlineSettings {
      bApplyOutlineToDropShadows=false
      bMiteredCorners=false
      bSeparateFillAlpha=false
      outlineMaterial=null
      outlineSize=0
      outlineColor=<rgba 0 0 0 1>
     }
    }
   }
   slot {class=ButtonSlot}
  }
  EditorUtilityButton_1 {
   widgetClassPath=&EditorUtilityButton
   parent=HorizontalBox_0
   widget=EditorUtilityButton_1
   slot {class=HorizontalBoxSlot, properties {padding=<ltrb 0 0 10 0>}}
  }
  TextBlock_1 {
   widgetClassPath=&TextBlock
   parent=EditorUtilityButton_1
   widget=TextBlock_1
   properties {
    text=Browse To
    font {
     bForceMonospaced=false
     bMaterialIsStencil=false
     fontMaterial=null
     fontObject=&Roboto
     letterSpacing=0
     monospacedWidth=1
     size=12
     skewAmount=0
     typefaceFontName=Bold
     outlineSettings {
      bApplyOutlineToDropShadows=false
      bMiteredCorners=false
      bSeparateFillAlpha=false
      outlineMaterial=null
      outlineSize=0
      outlineColor=<rgba 0 0 0 1>
     }
    }
   }
   slot {class=ButtonSlot}
  }
  EditorUtilityButton_52 {widgetClassPath=&EditorUtilityButton, parent=VerticalBox_29, widget=EditorUtilityButton_52, slot {class=VerticalBoxSlot}}
  TextBlock_0 {
   widgetClassPath=&TextBlock
   bIsVariable=true
   parent=EditorUtilityButton_52
   widget=TextBlock_0
   properties {text=Create VTIDs}
   slot {class=ButtonSlot}
  }
 ]
}
